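; BP_BattleStage_Platform is a platform Blueprint for the S8 battle stage environment.
; On BeginPlay it assigns its SM_Platform reference to the PlatformMesh variable, then calls the parent BeginPlay.
; Setting the variable before the parent call makes it available to the parent's startup logic.
; The construction script is empty, and the file shows no other behavior.
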(event EventBeginPlay
  (Variables|Default|SetPlatformMesh (Variables|Default|GetSM_Platform))
  (|CallParentFunction))

(fn ConstructionScript)
